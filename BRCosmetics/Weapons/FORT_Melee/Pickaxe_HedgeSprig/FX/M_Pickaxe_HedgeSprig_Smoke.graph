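# Translucent, unlit, two-sided smoke material for Niagara mesh particles in the HedgeSprig pickaxe effects.
# A soft band from a sphere mask, wave-distorted UVs and edge masks shape the smoke, tinted by particle color and faded by particle alpha and quality-switched depth fade.
# World position offset combines spline thickening, masked Perlin-noise wobble and a clamped pull toward a Niagara-fed position.
# Dynamic parameters drive width, wave shape, UV offset and movement.
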Material /BRCosmetics/Weapons/FORT_Melee/Pickaxe_HedgeSprig/FX/M_Pickaxe_HedgeSprig_Smoke
graph {
 properties {
  bUsedWithNiagaraMeshParticles=true
  blendMode=BLEND_Translucent
  pixelDepthOffsetMode=PDOM_Legacy
  shadingModel=MSM_Unlit
  twoSided=true
  usedShadingModels=Unlit
 }
 nodes [
  n1 = Constant()
  n2 = ScalarParameter() {parameterName=RandomWPO_Offset}
  n3 = VectorParameter() {
   group=WPO
   parameterName=WPO Depth UV
   sortPriority=6
   channelNames {a=WPO Depth V Pan, b=WPO Depth U Pan, g=WPO Depth V, r=WPO Depth U}
   defaultValue=<rgba 0 0.6 0 0.2>
  }
  n4 = Multiply(A=TextureCoordinate(), B=AppendVector(A=@n3.WPO Depth U, B=@n3.WPO Depth U))
  n5 = AppendVector(A=@n3.WPO Depth U Pan, B=@n3.WPO Depth U Pan)
  n6 = Add(A=Add(A=AppendVector(A=@n2, B=@n2), B=Panner(Coordinate=@n4, Speed=@n5))) {constB=0}
  n7 = Multiply(
   A=LinearInterpolate(
    A=ScalarParameter() {group=WPO, parameterName=WPO Z Clamp}
    Alpha=TextureSample(UVs=@n6) {samplerType=SAMPLERTYPE_LinearColor, texture=/Game/Global/Textures/Default/Tileable_Patterns/T_PerlinNoise_Uncompressed}.R
   ) {constB=-1}
   B=ScalarParameter() {defaultValue=100, group=WPO, parameterName=WPO Z Noise Amount}
  ) {constB=6}
  n8 = ComponentMask(None=TextureCoordinate()) {g=true}
  n9 = OneMinus(None=@n8)
  n10 = Constant() {r=1}
  n11 = Constant() {r=1}
  n12 = Multiply(A=Multiply(A=Power(Base=@n9, Exp=@n10) {constExponent=0.5}, B=Power(Base=@n8, Exp=@n11) {constExponent=0.5}), B=Constant() {r=2}) {constB=4}
  n13 = ScalarParameter() {defaultValue=3, group=WPO, parameterName=WPO Noise Intensity}
  Multiply(
   A=Multiply(
    A=MaterialFunctionCall(R=@n1, G=@n7, B=@n1) {materialFunction=&AppendMany}.RGB
    B=Saturate(None=@n12)
   )
   B=@n13
  )
  n14 = Multiply(B=ParticleColor().A) {constA=1}
  n15 = DepthFade(FadeDistance=ScalarParameter() {defaultValue=5, group=Opacity, parameterName=FadeDistance})
  n16 = Constant() {r=1}
  n17 = QualitySwitch(Default=@n15, Low=@n16, High=@n15, Medium=@n16, Epic=@n15)
  NamedRerouteDeclaration(None=Multiply(A=Multiply(A=@n14, B=@n17), B=NamedRerouteUsage())) {name=Shape, nodeColor=<rgba 1 0 0.3984 1>}
  n18 = MaterialFunctionCall(Float2=TextureCoordinate()) {materialFunction=&BreakOutFloat2Components}
  NamedRerouteDeclaration(None=@n18.G) {name=TopMask, nodeColor=<rgba 0.3359 1 0 1>}
  n19 = ScalarParameter() {defaultValue=8, group=SplineThicken, parameterName=SplineThickenWidth}
  n20 = Multiply(
   A=ScalarParameter() {defaultValue=5, group=SplineThicken, parameterName=ThickenCorePower}
   B=MaterialFunctionCall!(UVs, Offset, RotationCenter, RotationAngle=Constant() {r=0.25}) {materialFunction=/FortUILibrary/MaterialFunctions/Gradients/MF_UI_ReflectedGradient}.Result
  )
  n21 = Multiply(
   A=NamedRerouteDeclaration(None=Saturate(None=@n20)) {name=EdgeGradient, nodeColor=<rgba 1 0 0.9141 1>}
   B=ScalarParameter() {defaultValue=10, group=SplineThicken, parameterName=ThickenCoreSize}
  )
  n22 = ScalarParameter() {group=SplineThicken, parameterName=ThickenCore?}
  n23 = ScalarParameter() {defaultValue=3, group=SplineThicken, parameterName=SplineThickenWidthTip}
  n24 = MaterialFunctionCall(
   WidthBase=LinearInterpolate(A=@n19, B=Multiply(A=@n19, B=@n21), Alpha=@n22)
   WidthTip=LinearInterpolate(A=@n23, B=Multiply(A=@n21, B=@n23), Alpha=@n22)
  ) {materialFunction=&SplineThicken}
  n25 = Multiply(
   A=ScalarParameter() {defaultValue=1.3, group=SplineThicken, parameterName=Spline Width Mult}
   B=DynamicParameter() {defaultValue=<rgba 0 1 1 0>, paramNames [Random Offset, Spline Width Mult, Dissolve, Speed]}.Spline Width Mult
  )
  NamedRerouteDeclaration(None=StaticSwitchParameter(True=Multiply(A=@n24.WorldPositionOffset, B=@n25), False=Constant()) {defaultValue=true, group=SplineThicken, parameterName=UseSplineThicken}) {name=SplineThicken, nodeColor=<rgba 1 0.6797 0 1>}
  n26 = NamedRerouteUsage()
  n27 = Multiply(A=NamedRerouteUsage(), B=ScalarParameter() {defaultValue=1, parameterName=Param_9})
  NamedRerouteDeclaration(None=Add(A=Add(A=@n26, B=@n27) {constB=0}, B=NamedRerouteUsage()) {constB=0}) {name=WPO, nodeColor=<rgba 0.0781 0 1 1>}
  n28 = ParticleRandom()
  n29 = VectorParameter() {group=Distortion, parameterName=Distortion UV, sortPriority=1, defaultValue=<rgba 1 1 0 0.1>}
  n30 = ScalarParameter() {defaultValue=1, group=Noise, parameterName=Global Scale U}
  n31 = ScalarParameter() {defaultValue=2, group=Noise, parameterName=Global Scale V}
  n32 = Multiply(A=TextureCoordinate(), B=Multiply(A=AppendVector(A=@n29.R, B=@n29.R), B=AppendVector(A=@n30, B=@n31)))
  n33 = AppendVector(A=@n29.B, B=@n29.B)
  n34 = TextureSampleParameter2D(UVs=Add(A=AppendVector(A=@n28, B=@n28), B=Panner(Coordinate=@n32, Speed=@n33))) {
   group=Distortion
   parameterName="Distortion  Tex"
   samplerType=SAMPLERTYPE_Normal
   sortPriority=0
   texture=&T_Weap_Distort_N_02
  }
  NamedRerouteDeclaration(None=Multiply(
    A=ComponentMask(None=@n34.RGB) {g=true, r=true}
    B=ScalarParameter() {defaultValue=0.2, group=Distortion, parameterName=Distortion Intensity, sortPriority=2}
   )) {name=Distortion, nodeColor=<rgba 1 0 0.375 1>}
  NamedRerouteDeclaration(None=MaterialFunctionCall(Emissive=ParticleColor().RGB) {materialFunction=&MF_FX_EmissiveAdjustments}.Result) {name=Color, nodeColor=<rgba 1 0.375 0 1>}
  n35 = VectorParameter() {group=WPO, parameterName=WPO Noise UV, sortPriority=6, defaultValue=<rgba 0 0.65 0 0.1>}
  n36 = TextureCoordinate()
  n37 = AppendVector(A=ComponentMask(None=@n36) {r=true}, B=Power(Base=ComponentMask(None=@n36) {g=true}) {constExponent=1})
  n38 = Panner(Coordinate=Multiply(A=@n37, B=AppendVector(A=@n35.R, B=@n35.R)), Speed=AppendVector(A=@n35.B, B=@n35.B))
  n39 = ScalarParameter() {parameterName=RandomWPO_Offset}
  n40 = AppendVector(A=@n39, B=@n39)
  n41 = TextureObject() {samplerType=SAMPLERTYPE_LinearColor, texture=&T_PerlinNoise_Uncompressed}
  n42 = TextureSample(UVs=Add(A=Add(A=@n38, B=@n40)) {constB=0}, Tex=@n41) {samplerType=SAMPLERTYPE_LinearColor, texture=&T_PerlinNoise_Uncompressed}
  n43 = VectorParameter() {group=WPO, parameterName=WPO Noise UV 2, sortPriority=6, defaultValue=<rgba 0 0.5 0 0.2>}
  n44 = Panner(Coordinate=Multiply(A=@n37, B=AppendVector(A=@n43.R, B=@n43.R)), Speed=AppendVector(A=@n43.B, B=@n43.B))
  n45 = TextureSample(UVs=Add(A=Add(A=@n44, B=@n40)) {constB=0}, Tex=@n41) {samplerType=SAMPLERTYPE_LinearColor, texture=&T_PerlinNoise_Uncompressed}
  n46 = MaterialFunctionCall(
   R=Multiply(A=LinearInterpolate(Alpha=@n42.R) {constA=-1}, B=ScalarParameter() {defaultValue=1, group=WPO, parameterName=WPO X Noise Amount})
   G=Constant()
   B=Multiply(A=LinearInterpolate(Alpha=@n45.R) {constA=-1}, B=ScalarParameter() {defaultValue=1, group=WPO, parameterName=WPO Y Noise Amount}) {constB=0.5}
  ) {materialFunction=&AppendMany}
  n47 = ComponentMask(None=TextureCoordinate()) {g=true}
  n48 = Power(Base=OneMinus(None=@n47), Exp=Constant() {r=1.8}) {constExponent=0.5}
  n49 = LinearInterpolate(B=Power(Base=@n47, Exp=Constant() {r=0.9}) {constExponent=0.5}, Alpha=ScalarParameter() {group=SplineThicken, parameterName=WPO Top Mask}) {constA=1}
  n50 = Add(B=Multiply(A=Multiply(A=@n46.RGB, B=@n13), B=Multiply(A=@n48, B=@n49))) {constB=0}
  NamedRerouteDeclaration(None=Multiply(A=Transform(None=@n50), B=ScalarParameter() {defaultValue=1, parameterName=Param_8})) {name=WPO Distortion, nodeColor=<rgba 1 0 0.9844 1>}
  n51 = DynamicParameter() {desc=Parameters set by Niagara, parameterIndex=3, defaultValue=<rgba 0 1 1 1>, paramNames [X,Y,Z,Intensity]}
  n52 = MaterialFunctionCall(Float3=Subtract(
    A=MaterialFunctionCall(X=@n51.X, Y=@n51.X, Z=@n51.X) {materialFunction=&MakeFloat3}.Result
    B=ParticlePositionWS()
   )) {materialFunction=&BreakOutFloat3Components}
  n53 = ScalarParameter() {defaultValue=20, group=Interpolate, parameterName=Interpolate Clamp XY}
  n54 = ScalarParameter() {defaultValue=1, group=Interpolate, parameterName=Interpolate Clamp Z}
  n55 = AppendVector(
   A=Clamp(None=AppendVector(A=@n52.R, B=@n52.R), Min=Multiply(A=@n53) {constB=-1}, Max=@n53) {maxDefault=1000, minDefault=-1000}
   B=Clamp(None=@n52.B, Min=Multiply(A=@n54) {constB=-1}, Max=@n54) {minDefault=-1}
  )
  n56 = OneMinus(None=TextureCoordinate())
  n57 = Multiply(
   A=@n51.Intensity
   B=Power(Base=ComponentMask(None=@n56) {g=true}, Exp=ScalarParameter() {defaultValue=2, group=Interpolate, parameterName=Interpolate Power})
  )
  NamedRerouteDeclaration(None=StaticSwitchParameter(True=LinearInterpolate(B=@n55, Alpha=@n57), False=Constant()) {defaultValue=true, group=Interpolate, parameterName=Use Interpolate WPO, sortPriority=0}) {name=Interpolate WPO, nodeColor=<rgba 0.3828 1 0 1>}
  NamedRerouteDeclaration(None=MaterialFunctionCall(X=Constant(), Y=DynamicParameter() {defaultValue=<rgba 0 1 1 0>, paramNames [Random Offset, Spline Width Mult, Dissolve, Speed]}.Speed) {materialFunction=&MakeFloat2}.Result) {name=UV Additional Movement, nodeColor=<rgba 0 1 0.7422 1>}
  n58 = Constant() {r=0.4}
  n59 = MaterialFunctionCall(Float2=TextureCoordinate()) {materialFunction=&BreakOutFloat2Components}
  n60 = LinearInterpolate(A=Constant() {r=0.6}, B=Constant() {r=0.1}, Alpha=Power(Base=Multiply(A=@n59.G), Exp=Constant() {r=8}))
  n61 = MaterialFunctionCall(Float2=NamedRerouteUsage()) {materialFunction=&BreakOutFloat2Components}
  n62 = Constant() {r=0.5}
  n63 = Saturate(None=SmoothStep(Min=@n60, Max=Add(A=@n60, B=@n58), Value=SphereMask(A=@n61.R, B=@n62, Radius=@n62) {hardnessPercent=0}))
  n64 = NamedRerouteUsage()
  NamedRerouteDeclaration(None=Multiply(A=Multiply(A=@n63, B=@n64), B=NamedRerouteUsage())) {name=Base, nodeColor=<rgba 0 1 0.0625 1>}
  n65 = OneMinus(None=@n18.G)
  n66 = ScalarParameter() {defaultValue=5, parameterName=Param_5}
  NamedRerouteDeclaration(None=Saturate(None=Multiply(A=@n65, B=@n66))) {name=BottomMask, nodeColor=<rgba 0 1 0.7656 1>}
  n67 = MaterialFunctionCall(Float2=TextureCoordinate()) {materialFunction=&BreakOutFloat2Components}
  n68 = Constant() {r=3}
  n69 = Add(
   A=MaterialFunctionCall(X=@n67.R, Y=Power(Base=@n67.G, Exp=@n68)) {materialFunction=&MakeFloat2}.Result
   B=MaterialFunctionCall(
    X=Constant()
    Y=DynamicParameter() {defaultValue=<rgba 0 1 1 0>, paramNames [Random Offset, Spline Width Mult, Dissolve, Speed]}.Random Offset
   ) {materialFunction=&MakeFloat2}.Result
  )
  n70 = NamedRerouteUsage()
  n71 = Constant() {r=3}
  n72 = DynamicParameter() {parameterIndex=1, paramNames [Wave Amplitude, Wave Frequency, "", ""]}
  n73 = MaterialFunctionCall(Float2=TextureCoordinate()) {materialFunction=&BreakOutFloat2Components}
  n74 = Constant() {r=3}
  n75 = Multiply(A=ScalarParameter() {defaultValue=0.4, parameterName=WaveAmp}, B=OneMinus(None=Power(Base=@n73.G, Exp=@n74)))
  NamedRerouteDeclaration(None=MaterialFunctionCall!(
    UV=Add(A=@n69, B=@n70) {constB=0}
    U Frequency
    V Frequency=Multiply(A=@n71, B=@n72.Wave Frequency)
    U Amplitude
    V Amplitude=Multiply(A=@n75, B=@n72.Wave Amplitude)
    U Speed
    V Speed
    U Time=Constant()
    V Time
   ) {materialFunction=/Game/Effects/Fort_Effects/MaterialFunctions/MF_WaveUV}.Wave UVs) {name=WaveUV, nodeColor=<rgba 1 0.6094 0 1>}
  NamedRerouteUsage()
  NamedRerouteUsage()
  NamedRerouteUsage()
  NamedRerouteUsage()
  NamedRerouteUsage()
  NamedRerouteUsage()
  Transform(None=@n50) {transformSourceType=TRANSFORMSOURCE_Local}
 ]
}
