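# Physics asset for the HedgeSprig axe pickaxe, setting up jointed physics for its upper and lower stake bone chains.
# Capsule bodies cover the middle base and both two-segment stakes, with tiny zero-length capsules as each chain's root.
# Both roots lock angularly to the middle base.
# Stake joints set moderate first-swing limits, zero second-swing and slight twist, suggesting hinge-like bending in one plane.
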PhysicsAsset /BRCosmetics/Weapons/FORT_Melee/Pickaxe_HedgeSprig/Meshes/HedgeSprig_Axe_Physics
bodies {
 dyn_base_lower_00_stake [{shapeType=Capsule, length=0, radius=0.2, shapeName=dyn_base_lower_00_stake_capsule, rotation=<rot 0 -90 0>}]
 dyn_base_middle [{shapeType=Capsule, length=6.8921, radius=4.1162, center=<-0.0282 -0.1305 1.1425>, rotation=<rot -0.0021 -0.0466 4.7414>}]
 dyn_base_upper_stake [{shapeType=Capsule, length=0, radius=0.2, shapeName=dyn_base_upper_stake_capsule, rotation=<rot 0 -90 0>}]
 dyn_lower_stake [{shapeType=Capsule, length=8.6588, radius=4.0158, shapeName=dyn_lower_stake_capsule, center=<6.8238 0.0311 0.5768>}]
 dyn_lower_stake_end [{shapeType=Capsule, length=6.1012, radius=3.1306, shapeName=dyn_lower_stake_end_capsule, center=<6.4682 0.0597 -0.0654>}]
 dyn_upper_stake [{
   shapeType=Capsule
   length=9.8763
   radius=4.2015
   shapeName=dyn_upper_stake_capsule
   center=<5.6413 -0.0168 1.5518>
   rotation=<rot 0 0 11.0492>
  }]
 dyn_upper_stake_end [{shapeType=Capsule, length=8.3035, radius=3.2163, shapeName=dyn_upper_stake_end_capsule, center=<5.7139 0.0432 0.201>}]
}
constraints [
 {
  bone1Name=dyn_base_lower_00_stake
  bone2Name=dyn_base_middle
  swing1LimitDegrees=45
  swing1Motion=Locked
  swing2LimitDegrees=45
  swing2Motion=Locked
  twistLimitDegrees=45
  twistMotion=Locked
 }
 {bone1Name=dyn_lower_stake, bone2Name=dyn_base_lower_00_stake, swing1LimitDegrees=35, swing2LimitDegrees=0, twistLimitDegrees=5}
 {bone1Name=dyn_lower_stake_end, bone2Name=dyn_lower_stake, swing1LimitDegrees=30, swing2LimitDegrees=0, twistLimitDegrees=5}
 {
  bone1Name=dyn_base_upper_stake
  bone2Name=dyn_base_middle
  swing1LimitDegrees=45
  swing1Motion=Locked
  swing2LimitDegrees=45
  swing2Motion=Locked
  twistLimitDegrees=45
  twistMotion=Locked
 }
 {bone1Name=dyn_upper_stake, bone2Name=dyn_base_upper_stake, swing1LimitDegrees=40, swing2LimitDegrees=0, twistLimitDegrees=5}
 {bone1Name=dyn_upper_stake_end, bone2Name=dyn_upper_stake, swing1LimitDegrees=30, swing2LimitDegrees=0, twistLimitDegrees=5}
]
