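# Animation Blueprint for the HedgeSprig pickaxe that adds physics-driven secondary motion to its upper and lower stake bones.
# Outside the front end, a chain of spring-based AnimDynamics nodes drives the stakes from the cached input pose.
# In the front end, bRigidBodyEnabled picks RigidBody or AnimDynamics, with separate tunings for held and not held.
# The held flag follows whether the Body Mesh cached on customization load is valid.
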AnimBlueprint /BRCosmetics/Weapons/FORT_Melee/Pickaxe_HedgeSprig/Meshes/HedgeSprig_Axe_AnimBP
target_skeleton=&/HedgeSprig_Axe_Skeleton
graphs [
 AnimGraph {nodes [
   n1 = Anim Dynamics - Bone: dyn_upper_stake(ComponentPose=Local To Component(LocalPose=Use cached pose 'input pose'()), Alpha=1, ExternalForce="0,0,0") {
    angularDampingOverride=0.9999
    angularSpringConstant=10000
    bAngularSpring=true
    bOverrideAngularDamping=true
    bUseGravityOverride=true
    bUsePlanarLimit=false
    simulationSpace=World
    alphaBoolBlend {blendOption=Linear}
    boundBone {boneName=dyn_upper_stake}
    componentAppliedLinearAccClamp=<10000 10000 10000>
    componentLinearAccScale=<1 1 1>
    componentLinearVelScale=<1 1 1>
    gravityOverride {z=-980}
    physicsBodyDefinitions [{
      boundBone {boneName=dyn_upper_stake}
      boxExtents=<200 200 200>
      constraintSetup {angularLimitsMax=<0.1 0.1 0.1>, angularLimitsMin=<0.1 0.1 0.1>, angularTarget=<1 0 0>}
      localJointOffset=<5 0 0>
     }]
    relativeSpaceBone {boneName=dyn_base_middle}
   }
   n2 = Anim Dynamics - Bone: dyn_lower_stake(ComponentPose=@n1, Alpha=1, ExternalForce="0,0,0") {
    angularDampingOverride=0.9999
    angularSpringConstant=10000
    bAngularSpring=true
    bOverrideAngularDamping=true
    bUseGravityOverride=true
    simulationSpace=World
    alphaBoolBlend {blendOption=Linear}
    boundBone {boneName=dyn_lower_stake}
    componentAppliedLinearAccClamp=<10000 10000 10000>
    componentLinearAccScale=<1 1 1>
    componentLinearVelScale=<1 1 1>
    gravityOverride {z=-980}
    physicsBodyDefinitions [{
      boundBone {boneName=dyn_lower_stake}
      boxExtents=<100 100 100>
      constraintSetup {angularLimitsMax=<0.1 0.1 0.1>, angularLimitsMin=<0.1 0.1 0.1>, angularTarget=<1 0 0>}
      localJointOffset=<3 0 0>
     }]
    relativeSpaceBone {boneName=dyn_base_middle}
   }
   n3 = RigidBody(ComponentPose=Local To Component(LocalPose=Use cached pose 'input pose'())) {
    bOverrideWorldGravity=true
    cachedBoundsScale=2
    overridePhysicsAsset=&/HedgeSprig_Axe_Physics
    simulationSpace=BaseBoneSpace
    alphaBoolBlend {blendOption=Linear}
    baseBoneRef {boneName=b_AR_Root}
    componentAppliedLinearAccClamp=<100 100 100>
    overrideWorldGravity {z=-980}
    simSpaceSettings {
     dampingAlpha=4
     maxAngularAcceleration=3
     maxAngularVelocity=5
     maxLinearAcceleration=100
     maxLinearVelocity=100
     velocityScaleZ=10
     worldAlpha=0.1
    }
   }
   n4 = Blend Poses by bool(BlendPose_0=Component To Local(ComponentPose=@n3), BlendPose_1=Component To Local(ComponentPose=@n2), bActiveValue=Get bRigidBodyEnabled()) {blendType=Linear}
   n5 = Anim Dynamics - Bone: dyn_upper_stake(ComponentPose=Local To Component(LocalPose=Use cached pose 'input pose'()), Alpha=1, ExternalForce="0,0,0") {
    angularDampingOverride=0.9999
    angularSpringConstant=3000
    bAngularSpring=true
    bOverrideAngularDamping=true
    bUseGravityOverride=true
    alphaBoolBlend {blendOption=Linear}
    boundBone {boneName=dyn_upper_stake}
    componentAppliedLinearAccClamp=<1000 1000 2000>
    componentLinearAccScale=<0.3 0.3 0.3>
    componentLinearVelScale=<0.4 0.4 0.4>
    gravityOverride {z=-980}
    physicsBodyDefinitions [{
      boundBone {boneName=dyn_upper_stake}
      boxExtents=<30 30 30>
      constraintSetup {angularLimitsMax=<2 0 50>, angularLimitsMin=<-2 0 -50>, angularTarget=<1 0 0>}
      localJointOffset=<3 -0.5 0>
     }]
    relativeSpaceBone {boneName=dyn_base_middle}
   }
   n6 = Anim Dynamics - Bone: dyn_upper_stake_end(ComponentPose=@n5, Alpha=1, ExternalForce="0,0,0") {
    angularDampingOverride=0.999
    angularSpringConstant=3000
    bAngularSpring=true
    bOverrideAngularDamping=true
    bUseGravityOverride=true
    alphaBoolBlend {blendOption=Linear}
    boundBone {boneName=dyn_upper_stake_end}
    componentAppliedLinearAccClamp=<1000 1000 2000>
    componentLinearAccScale=<0.3 0.3 0.3>
    componentLinearVelScale=<0.4 0.4 0.4>
    gravityOverride {z=-980}
    physicsBodyDefinitions [{
      boundBone {boneName=dyn_upper_stake_end}
      boxExtents=<20 20 20>
      constraintSetup {angularLimitsMax=<0 5 50>, angularLimitsMin=<0 -5 -50>, angularTarget=<1 0 0>}
      localJointOffset=<1 -0.2 0>
     }]
    relativeSpaceBone {boneName=dyn_base_middle}
   }
   n7 = Anim Dynamics - Bone: dyn_lower_stake(ComponentPose=@n6, Alpha=1, ExternalForce="0,0,0") {
    angularDampingOverride=0.99
    angularSpringConstant=5000
    bAngularSpring=true
    bOverrideAngularDamping=true
    bUseGravityOverride=true
    alphaBoolBlend {blendOption=Linear}
    boundBone {boneName=dyn_lower_stake}
    componentAppliedLinearAccClamp=<1000 1000 2000>
    componentLinearAccScale=<0.4 0.4 0.4>
    componentLinearVelScale=<0.5 0.5 0.5>
    gravityOverride {z=-980}
    physicsBodyDefinitions [{
      boundBone {boneName=dyn_lower_stake}
      boxExtents=<30 30 30>
      constraintSetup {angularLimitsMax=<2 0 10>, angularLimitsMin=<-2 0 -10>, angularTarget=<1 0 0>}
      localJointOffset=<2 -1.3 0>
     }]
    relativeSpaceBone {boneName=dyn_base_middle}
   }
   n8 = Anim Dynamics - Bone: dyn_lower_stake_end(ComponentPose=@n7, Alpha=1, ExternalForce="0,0,0") {
    angularDampingOverride=0.999
    angularSpringConstant=5000
    bAngularSpring=true
    bOverrideAngularDamping=true
    bUseGravityOverride=true
    alphaBoolBlend {blendOption=Linear}
    boundBone {boneName=dyn_lower_stake_end}
    componentAppliedLinearAccClamp=<1000 1000 2000>
    componentLinearAccScale=<0.4 0.4 0.4>
    componentLinearVelScale=<0.5 0.5 0.5>
    gravityOverride {z=-980}
    physicsBodyDefinitions [{
      boundBone {boneName=dyn_lower_stake_end}
      boxExtents=<20 20 20>
      constraintSetup {angularLimitsMax=<5 5 30>, angularLimitsMin=<-5 -5 -30>, angularTarget=<1 0 0>}
      localJointOffset=<2 -0.5 0>
     }]
    relativeSpaceBone {boneName=dyn_base_middle}
   }
   n9 = RigidBody(ComponentPose=Local To Component(LocalPose=Use cached pose 'input pose'())) {
    bOverrideWorldGravity=true
    overridePhysicsAsset=&HedgeSprig_Axe_Physics
    simulationSpace=BaseBoneSpace
    alphaBoolBlend {blendOption=Linear}
    baseBoneRef {boneName=Melee_TwoHanded_Handle}
    overrideWorldGravity {z=-980}
    simSpaceSettings {
     dampingAlpha=0.5
     maxAngularAcceleration=50
     maxAngularVelocity=8
     maxLinearAcceleration=1200
     maxLinearVelocity=500
     velocityScaleZ=0.8
     worldAlpha=1
    }
   }
   n10 = Blend Poses by bool(BlendPose_0=Component To Local(ComponentPose=@n9), BlendPose_1=Component To Local(ComponentPose=@n8), bActiveValue=Get bRigidBodyEnabled()) {blendType=Linear}
   n11 = Get Character Holding Pickaxe() -> (Character Holding Pickaxe:bool)
   n12 = Anim Dynamics - Bone: dyn_upper_stake(ComponentPose=Local To Component(LocalPose=Use cached pose 'input pose'()), Alpha=1, ExternalForce="0,0,0") {
    angularDampingOverride=0.9999
    angularSpringConstant=5000
    bAngularSpring=true
    bOverrideAngularDamping=true
    bUseGravityOverride=true
    alphaBoolBlend {blendOption=Linear}
    boundBone {boneName=dyn_upper_stake}
    componentAppliedLinearAccClamp=<1000 900 700>
    componentLinearAccScale=<0.3 0.3 0.3>
    componentLinearVelScale=<0.4 0.4 0.4>
    gravityOverride {z=-980}
    physicsBodyDefinitions [{
      boundBone {boneName=dyn_upper_stake}
      boxExtents=<15 20 15>
      constraintSetup {angularLimitsMax=<2 10 50>, angularLimitsMin=<-2 -10 -10>, angularTarget=<1 0 0>}
      localJointOffset=<3 1.5 0>
     }]
    relativeSpaceBone {boneName=dyn_base_middle}
   }
   n13 = Anim Dynamics - Bone: dyn_upper_stake_end(ComponentPose=@n12, Alpha=1, ExternalForce="0,0,0") {
    angularDampingOverride=0.999
    angularSpringConstant=5000
    bAngularSpring=true
    bOverrideAngularDamping=true
    bUseGravityOverride=true
    alphaBoolBlend {blendOption=Linear}
    boundBone {boneName=dyn_upper_stake_end}
    componentAppliedLinearAccClamp=<1000 900 700>
    componentLinearAccScale=<0.3 0.3 0.3>
    componentLinearVelScale=<0.4 0.4 0.4>
    gravityOverride {z=-980}
    physicsBodyDefinitions [{
      boundBone {boneName=dyn_upper_stake_end}
      boxExtents=<15 20 15>
      constraintSetup {angularLimitsMax=<5 10 50>, angularLimitsMin=<-5 -10 -30>, angularTarget=<1 0 0>}
      localJointOffset=<3 1 0>
     }]
    relativeSpaceBone {boneName=dyn_base_middle}
   }
   n14 = Anim Dynamics - Bone: dyn_lower_stake(ComponentPose=@n13, Alpha=1, ExternalForce="0,0,0") {
    angularDampingOverride=0.9999
    angularSpringConstant=5500
    bAngularSpring=true
    bOverrideAngularDamping=true
    bUseGravityOverride=true
    alphaBoolBlend {blendOption=Linear}
    boundBone {boneName=dyn_lower_stake}
    componentAppliedLinearAccClamp=<1000 1000 1000>
    componentLinearAccScale=<0.2 0.3 0.2>
    componentLinearVelScale=<0.3 0.4 0.3>
    gravityOverride {z=-980}
    physicsBodyDefinitions [{
      boundBone {boneName=dyn_lower_stake}
      boxExtents=<12 20 15>
      constraintSetup {angularLimitsMax=<2 10 20>, angularLimitsMin=<-2 -10 -50>, angularTarget=<1 0 0>}
      localJointOffset=<3 -1.5 0>
     }]
    relativeSpaceBone {boneName=dyn_base_middle}
   }
   n15 = Anim Dynamics - Bone: dyn_lower_stake_end(ComponentPose=@n14, Alpha=1, ExternalForce="0,0,0") {
    angularDampingOverride=0.999
    angularSpringConstant=5000
    bAngularSpring=true
    bOverrideAngularDamping=true
    bUseGravityOverride=true
    alphaBoolBlend {blendOption=Linear}
    boundBone {boneName=dyn_lower_stake_end}
    componentAppliedLinearAccClamp=<1000 900 700>
    componentLinearAccScale=<0.2 0.3 0.3>
    componentLinearVelScale=<0.3 0.4 0.4>
    gravityOverride {z=-980}
    physicsBodyDefinitions [{
      boundBone {boneName=dyn_lower_stake_end}
      boxExtents=<15 20 15>
      constraintSetup {angularLimitsMax=<5 10 30>, angularLimitsMin=<-5 -10 -30>, angularTarget=<1 0 0>}
      localJointOffset=<3 -1 0>
     }]
    relativeSpaceBone {boneName=dyn_base_middle}
   }
   Output Pose(Result=Blend Poses by bool(
     BlendPose_0=Blend Poses by bool(BlendPose_0=@n10, BlendPose_1=@n4, bActiveValue=@n11) {blendType=Linear}
     BlendPose_1=Component To Local(ComponentPose=@n15)
     bActiveValue=Get bIsInFrontEnd()
    ) {blendType=Linear}) {layerGroup=None}
   AnimGraphNode_SubInput_0 = Input Pose()
   Save cached pose 'input pose'(Pose=@AnimGraphNode_SubInput_0) {cacheName=input pose}
  ]}
 EventGraph {nodes [
   Event BlueprintUpdateAnimation()
   > n1 = Get!(execute:Exec) -> (then:Exec, else:Exec, Body Mesh:Skeletal Mesh Component Budgeted Object Reference)
   > Set Character Holding Pickaxe(Character Holding Pickaxe:bool=true) -> (Output_Get:bool)
   Set Character Holding Pickaxe(execute=@n1.else, Character Holding Pickaxe=false)
   n2 = Event OnCharacterCustomizationLoaded()
   > Set Body Mesh(Body Mesh:Skeletal Mesh Component Budgeted Object Reference=Get Array Item with EFort Custom Part Type(Array=@n2.CustomMeshParts)) -> (Output_Get:Skeletal Mesh Component Budgeted Object Reference)
   TryGetPawnOwner()
  ]}
]
variables [
 Body Mesh {type=SkeletalMeshComponentBudgeted}
 Character Holding Pickaxe {type=boolean}
]
